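# Lit, opaque material for the Creative diamond resource prop, faking gem depth with glowing inner layers.
# Two BumpOffset samples of an offset mask form inner and outer layers, blended with glow colors by a view-angle falloff and contrasted with an S-curve.
# A normal-mapped Fresnel adds colored rim glow; mesh masks steer roughness, edges, separation tint and normal blending.
# The attributes pass through time-of-day emissive and pickup hologram functions.
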Material /Game/Creative/Items/Gameplay/Resources/Materials/M_CM_Prop_Diamond
graph {
 properties {
  bUseMaterialAttributes=true
  bUsedWithInstancedStaticMeshes=true
  bUsedWithNiagaraMeshParticles=true
  pixelDepthOffsetMode=PDOM_Legacy
  translucencyLightingMode=TLM_Surface
  usedShadingModels=Default Lit
 }
 nodes [
  n1 = TextureSampleParameter2D(UVs=TextureCoordinate()) {
   parameterName=Bump Normal Map
   samplerType=SAMPLERTYPE_Normal
   texture=/Game/Creative/Items/Gameplay/Resources/Textures/T_CM_Prop_DiamondOffset_N
  }
  n2 = ScalarParameter() {defaultValue=0.5, group=Frens_Ext, parameterName=NormalMap_Fresnel, sortPriority=6}
  n3 = Fresnel(
   ExponentIn=ScalarParameter() {defaultValue=0.5, group=Frens_Ext, parameterName=Fre_Exp_B}
   BaseReflectFractionIn=ScalarParameter() {defaultValue=1, group=Frens_Ext, parameterName=Fre_power_B}
   Normal=Transform(None=MaterialFunctionCall(Normal=@n1.RGB, Flatness=@n2) {materialFunction=&FlattenNormal}.Result)
  )
  n4 = TextureSampleParameter2D() {parameterName=Mesh Mask, samplerType=SAMPLERTYPE_Masks, texture=/Game/Creative/Items/Gameplay/Resources/Textures/T_CM_Prop_Diamond_S}
  n5 = Multiply(
   A=Multiply(A=@n3, B=ScalarParameter() {defaultValue=0.1, group=Frens_Ext, parameterName=FrensAdd_Strength})
   B=LinearInterpolate(Alpha=@n4.R) {constA=1, constB=0}
  )
  n6 = ScalarParameter() {group=Roughness, parameterName=Roughness_low}
  n7 = ScalarParameter() {defaultValue=1, group=Roughness, parameterName=Roughness_high}
  n8 = Add(A=Multiply(A=@n3) {constB=0.25}, B=Saturate(None=LinearInterpolate(A=@n6, B=@n7, Alpha=@n4.R)))
  n9 = LinearInterpolate(A=ScalarParameter() {defaultValue=1, group=Color, parameterName=White Edge Strength}, Alpha=OneMinus(None=@n4.G)) {constA=1, constB=0.5}
  n10 = VectorParameter() {group=Color, parameterName=ColorGlow_outer, defaultValue=<rgba 0 0.0344 0.1458 1>}
  n11 = BumpOffset(Coordinate=TextureCoordinate(), Height=ScalarParameter() {defaultValue=-0.75, group=BumpOffset, parameterName=BO_Height_inner})
  n12 = Add(A=OneMinus(None=TextureSampleParameter2D(UVs=@n11) {parameterName=Bump Mask, texture=/Game/Creative/Items/Gameplay/Resources/Textures/T_CM_Prop_DiamondOffset_S}.R)) {constB=-0.8}
  n13 = Multiply(
   A=VectorParameter() {group=Color, parameterName=ColorGlow_inner, defaultValue=<rgba 0.1695 0.2808 0.651 1>}.RGB
   B=LinearInterpolate(Alpha=Saturate(None=@n12)) {constA=0.1}
  )
  n14 = VertexNormalWS()
  n15 = CameraVectorWS()
  n16 = Multiply(
   A=Power(Base=DotProduct(A=@n14, B=@n15), Exp=ScalarParameter() {defaultValue=2.5, group=Frens_Int, parameterName=Fre_ExpC})
   B=ScalarParameter() {defaultValue=1, group=Frens_Int, parameterName=Fre_powerC}
  )
  n17 = Multiply(A=LinearInterpolate(Alpha=Saturate(None=@n16)) {constA=0.05}, B=OneMinus(None=@n4.R))
  n18 = TextureCoordinate() {uTiling=2, vTiling=2}
  n19 = ScalarParameter() {defaultValue=-0.25, group=BumpOffset, parameterName=BO_Height_Outer}
  n20 = Add(A=TextureSampleParameter2D(UVs=BumpOffset(Coordinate=@n18, Height=@n19)) {parameterName=Bump Mask, texture=&T_CM_Prop_DiamondOffset_S}.G) {constB=-0.94}
  n21 = Add(
   A=MaterialFunctionCall(
    In=LinearInterpolate(A=@n10.RGB, B=@n13, Alpha=@n17)
    Power=ScalarParameter() {defaultValue=1, group=BumpOffset, parameterName=Interior Contrast, sliderMin=1}
   ) {materialFunction=&SCurve}.Result
   B=Multiply(A=Saturate(None=@n20), B=VectorParameter() {group=BumpOffset, parameterName=BO_ColorOuter, defaultValue=<rgba 0.1826 0.2384 0.5885 1>}.RGB)
  )
  n22 = Add(
   A=Multiply(A=VectorParameter() {group=Color, parameterName=Normal Fresnel Color, defaultValue=<rgba 1.5302 0.7188 2 1>}.RGB, B=Divide(A=@n5))
   B=Multiply(A=Multiply(A=@n9, B=@n21), B=ScalarParameter() {defaultValue=3, group=BumpOffset, parameterName=Emissive_Strength})
  )
  n23 = VectorParameter() {group=Color, parameterName=Separation Color, defaultValue=<rgba 0.1695 0.2808 0.651 1>}
  n24 = Multiply(A=@n4.A, B=ScalarParameter() {defaultValue=0.6, group=Color, parameterName=Separation Strength}) {constB=0.6}
  n25 = TextureSampleParameter2D() {parameterName=Mesh Normal Map, samplerType=SAMPLERTYPE_Normal, texture=/Game/Creative/Items/Gameplay/Resources/Textures/T_CM_Prop_Diamond_N}
  n26 = ScalarParameter() {defaultValue=0.65, group=Detail Normal, parameterName=NormalFlatten, sortPriority=6}
  n27 = OneMinus(None=@n4.G)
  n28 = SetMaterialAttributes!(
   MaterialAttributes
   Base Color=@n5
   Metallic=Constant()
   Specular=OneMinus(None=@n8)
   Roughness=@n8
   Emissive Color=Multiply(
    A=LinearInterpolate(A=@n22, B=@n23.RGB, Alpha=@n24) {constB=0.15}
    B=ScalarParameter() {defaultValue=3, group=Color, parameterName=Color Brightness}
   ) {constB=3}
   Normal=LinearInterpolate(
    A=@n25.RGB
    B=MaterialFunctionCall(Normal=@n1.RGB, Flatness=@n26) {materialFunction=&FlattenNormal}.Result
    Alpha=Power(Base=@n27)
   )
  ) {attributeSetTypes [
    #1
    #2
    #3
    #4
    #5
    #6
   ]}
  MaterialFunctionCall!(In=MaterialFunctionCall!(In=@n28) {materialFunction=/Game/Athena/WinterArt/Materials/MF_TOD_EmissiveColors}.Result) {materialFunction=/Game/Packages/Fortress_SharedMaterials/Base_Material_Functions/MF_Pickup_Hologram}
 ]
}
