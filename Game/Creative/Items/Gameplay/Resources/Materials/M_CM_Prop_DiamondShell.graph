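# Translucent, lit material for the diamond shell of a Creative resource prop, giving a tinted, faceted, glassy look with reflections; also usable on Niagara mesh particles.
# A brightened, desaturated cubemap combines with Color for base color and emissive, while Fresnel drives opacity and index-of-refraction.
# World-position derivatives yield faceted normals, swapped for a flat normal map in the material proxy.
# MF_CreativeMoveTool and MF_Pickup_Hologram wrap the result.
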Material /Game/Creative/Items/Gameplay/Resources/Materials/M_CM_Prop_DiamondShell
graph {
 properties {
  bUseMaterialAttributes=true
  bUsedWithNiagaraMeshParticles=true
  blendMode=BLEND_Translucent
  pixelDepthOffsetMode=PDOM_Legacy
  refractionMethod=RM_IndexOfRefraction
  usedShadingModels=Default Lit
 }
 nodes [
  n1 = ScalarParameter() {parameterName=Opacity_intensity}
  n2 = Fresnel(ExponentIn=ScalarParameter() {parameterName=Fresnel_Exponent}, BaseReflectFractionIn=ScalarParameter() {parameterName=Fresnel_Base})
  n3 = WorldPosition()
  n4 = DDX(Value=@n3.XYZ)
  n5 = DDY(Value=@n3.XYZ)
  n6 = Constant() {r=1}
  n7 = ScalarParameter() {parameterName=ReFraction_Adjust}
  n8 = TextureSampleParameterCube!(UVs=ReflectionVectorWS(), Apply View MipBias) {
   parameterName=Reflection Cubemap
   samplerType=SAMPLERTYPE_LinearColor
   texture=/Game/Creative/Environments/Meshes/Glass/Textures/T_Glass_Cubemap_01
  }
  n9 = Desaturation(None=Multiply(A=@n8.RGB, B=ScalarParameter() {parameterName=Cube_brightness}), Fraction=ScalarParameter() {parameterName=Cube_desaturation})
  n10 = VectorParameter() {parameterName=Color, defaultValue=<rgba 0.2962 0.4347 0.9896 1>}
  n11 = Multiply(A=@n9, B=@n10.RGB)
  n12 = ScalarParameter() {defaultValue=1, parameterName=Emissive Multiplier}
  n13 = MaterialFunctionCall(MA In=MakeMaterialAttributes!(
    BaseColor=Add(A=@n9, B=@n10.RGB)
    Metallic=Constant()
    Specular=ScalarParameter() {parameterName=SPec}
    Roughness=ScalarParameter() {parameterName=Roughness}
    Anisotropy
    EmissiveColor=Multiply(A=@n11, B=@n12)
    Opacity=Add(A=@n1, B=@n2)
    OpacityMask
    Normal=CrossProduct(A=@n4, B=@n5)
    Tangent
    WorldPositionOffset
    SubsurfaceColor
    ClearCoat
    ClearCoatRoughness
    AmbientOcclusion
    Refraction=LinearInterpolate(A=@n6, B=@n7, Alpha=@n2)
    CustomizedUVs_0
    CustomizedUVs_1
    CustomizedUVs_2
    CustomizedUVs_3
    CustomizedUVs_4
    CustomizedUVs_5
    CustomizedUVs_6
    CustomizedUVs_7
    PixelDepthOffset
    ShadingModel
    Displacement
   )) {materialFunction=&MF_CreativeMoveTool}
  n14 = TextureSample() {samplerType=SAMPLERTYPE_Normal, texture=&FlatNormalBlue}
  MaterialFunctionCall!(In=MaterialProxyReplace(
    Realtime=@n13.MA Out
    MaterialProxy=SetMaterialAttributes!(MaterialAttributes=@n13.MA Out, Normal=@n14.RGB) {attributeSetTypes [#1]}
   )) {materialFunction=/Game/Packages/Fortress_SharedMaterials/Base_Material_Functions/MF_Pickup_Hologram}
 ]
}
