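# Subsurface material for the Creative grub item prop, giving a fleshy look with inner glow, routed through MF_Pickup_Hologram.
# Base color blends Fresnel inner and edge glow with tinted dirt and grub diffuse, using a darkened bump-offset copy for depth.
# Alpha textures drive opacity and subsurface color, and tiled skin normals blend onto the base normal.
# Glow feeds base color, not emissive; roughness and specular adjust blends are unconnected.
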Material /Game/Creative/Items/Gameplay/Resources/Materials/M_CP_Grub_SS_03
graph {
 properties {bUseMaterialAttributes=true, pixelDepthOffsetMode=PDOM_Legacy, shadingModel=MSM_Subsurface, usedShadingModels=Subsurface}
 nodes [
  n1 = TextureSampleParameter2D() {
   group=Textures
   parameterName=PMSK_SMR
   samplerType=SAMPLERTYPE_Masks
   texture=/Game/Creative/Items/Gameplay/Resources/Textures/T_Raptor_eye_Dirt_S
  }
  n2 = TextureSample() {samplerType=SAMPLERTYPE_Masks, texture=/Game/Creative/Items/Gameplay/Resources/Textures/T_CM_Prop_Grub_01_S}
  LinearInterpolate(
   A=Multiply(A=@n1.B, B=ScalarParameter() {parameterName=Roughness_adjust})
   B=@n2.B
   Alpha=ScalarParameter() {parameterName=Rough_adjust_post}
  )
  LinearInterpolate(
   A=Multiply(A=@n1.R, B=ScalarParameter() {defaultValue=1, parameterName=Specular_adjust})
   B=@n2.R
   Alpha=ScalarParameter() {parameterName=Spec_adjust_post}
  )
  n3 = ScalarParameter() {defaultValue=9, group=InnerGlow, parameterName=Fresnel-Exp_InteriorGlow}
  n4 = PixelNormalWS()
  n5 = Multiply(A=Multiply(
    A=VectorParameter() {group=InnerGlow, parameterName=Color-Emissive_InteriorGlow, defaultValue=<rgba 1.5 0.4636 0.1471 1>}.RGB
    B=MaterialFunctionCall(Exponent=@n3, Vertex / Pixel Normal=@n4) {materialFunction=&MF_Fresnel_FX}.Fresnel
   ))
  n6 = VectorParameter() {group=InnerGlow, parameterName=Color-EdgeGlow_InteriorGlow, defaultValue=<rgba 10 3.1644 0.5415 1>}
  n7 = MaterialFunctionCall(
   Brightness=ScalarParameter() {defaultValue=1.2, group=InnerGlow, parameterName=Edge-Brightness_InteriorGlow, sliderMin=-0.008}
   Exponent=ScalarParameter() {defaultValue=4, group=InnerGlow, parameterName=Edge-Exp_InteriorGlow}
   Vertex / Pixel Normal=@n4
  ) {materialFunction=&MF_Fresnel_FX}
  n8 = LinearInterpolate(A=Add(A=Multiply(A=@n5), B=Multiply(A=@n6.RGB, B=@n7.Fresnel Inverted)), B=Constant()) {constB=0}
  n9 = Multiply(A=TextureCoordinate(), B=ScalarParameter() {parameterName=Dirt_tile_Rate})
  n10 = TextureSampleParameter2D() {
   group=Textures
   parameterName=Top_masks_Emissive_Masks
   samplerType=SAMPLERTYPE_Masks
   texture=&T_Raptor_eye_Dirt_S
  }
  n11 = LinearInterpolate(
   B=Multiply(
    A=TextureSampleParameter2D(UVs=@n9) {group=Textures, parameterName=ForeGround_image, texture=/Game/Creative/Items/Gameplay/Resources/Textures/T_Raptor_eye_Dirt}.RGB
    B=VectorParameter() {group=property_adjust, parameterName=Color_Top_Tint, defaultValue=<rgba 0.8 0.1991 0.071 1>}.RGB
   )
   Alpha=@n10.B
  )
  n12 = Constant() {r=0.2}
  n13 = TextureSample(UVs=BumpOffset(Coordinate=TextureCoordinate(), Height=ScalarParameter() {defaultValue=7.25, group=Offset_adjust, parameterName=BackOffset}) {heightRatio=0.1, referencePlane=8}) {texture=/Game/Creative/Items/Gameplay/Resources/Textures/T_CM_Prop_Grub_01_D}
  n14 = Multiply(A=@n13.RGB, B=ScalarParameter() {defaultValue=0.15, group=property_adjust, parameterName=darken_back})
  n15 = TextureSample() {texture=&T_CM_Prop_Grub_01_D}
  n16 = TextureSampleParameter2D(UVs=TextureCoordinate()) {parameterName=alpha _SS, samplerType=SAMPLERTYPE_LinearColor, texture=/Game/Creative/Items/Gameplay/Resources/Textures/T_CM_Grub_01_A2}
  n17 = OneMinus(None=@n16.RGB)
  n18 = LinearInterpolate(
   A=LinearInterpolate(A=Add(A=@n11, B=@n12), B=LinearInterpolate(A=@n14, B=@n15.RGB, Alpha=@n17), Alpha=OneMinus(None=@n10.B))
   B=@n13.RGB
   Alpha=OneMinus(None=@n16.RGB)
  )
  n19 = Constant() {r=0.5}
  n20 = Multiply(A=@n1.G, B=ScalarParameter() {parameterName=metallic_adjust})
  n21 = ScalarParameter() {parameterName=Metal_adjust_post}
  n22 = ScalarParameter() {defaultValue=0.5}
  n23 = ScalarParameter() {defaultValue=1, parameterName=SS_image_Strength}
  n24 = TextureSample() {samplerType=SAMPLERTYPE_Normal, texture=/Game/Creative/Items/Gameplay/Resources/Textures/T_CM_Prop_Grub_01_N}
  n25 = Multiply(A=TextureCoordinate(), B=ScalarParameter() {defaultValue=6, group=Glass, parameterName=Normal_TileRate})
  n26 = TextureCoordinate()
  n27 = ScalarParameter() {defaultValue=6, group=Glass, parameterName=Normal_TileRate2}
  n28 = LinearInterpolate(
   A=MaterialFunctionCall(
    Normal=TextureSample(UVs=@n25) {samplerType=SAMPLERTYPE_Normal, texture=/Game/Creative/Items/Gameplay/Resources/Textures/T_Grub_Skin_N}.RGB
    Flatness=ScalarParameter() {defaultValue=1.15, group=Glass, parameterName=NormalFlatten}
   ) {materialFunction=&FlattenNormal}.Result
   B=TextureSample(UVs=Multiply(A=@n26, B=@n27)) {samplerType=SAMPLERTYPE_Normal, texture=&T_Grub_Skin_N}.RGB
  )
  n29 = Multiply(A=VectorParameter() {parameterName=subsurface, defaultValue=<rgba 0.5833 0.0055 0 1>}.RGB, B=Constant() {r=1})
  n30 = TextureSample() {samplerType=SAMPLERTYPE_LinearColor, texture=/Game/Creative/Items/Gameplay/Resources/Textures/T_CM_Grub_01_A}
  MaterialFunctionCall!(In=MakeMaterialAttributes!(
    BaseColor=LinearInterpolate(A=@n8, B=@n18, Alpha=@n19)
    Metallic=LinearInterpolate(A=@n20, B=@n2.G, Alpha=@n21)
    Specular=@n2.R
    Roughness=Multiply(A=@n2.B, B=@n22)
    Anisotropy
    EmissiveColor
    Opacity=Multiply(A=@n16.RGB, B=@n23)
    OpacityMask
    Normal=MaterialFunctionCall(BaseNormal=@n24.RGB, AdditionalNormal=@n28) {materialFunction=&BlendAngleCorrectedNormals}.Result
    Tangent
    WorldPositionOffset
    SubsurfaceColor=LinearInterpolate(A=@n29, B=@n15.RGB, Alpha=@n30.RGB)
    ClearCoat
    ClearCoatRoughness
    AmbientOcclusion
    Refraction
    CustomizedUVs_0
    CustomizedUVs_1
    CustomizedUVs_2
    CustomizedUVs_3
    CustomizedUVs_4
    CustomizedUVs_5
    CustomizedUVs_6
    CustomizedUVs_7
    PixelDepthOffset
    ShadingModel
    Displacement
   )) {materialFunction=/Game/Packages/Fortress_SharedMaterials/Base_Material_Functions/MF_Pickup_Hologram}
  OneMinus()
 ]
}
