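# Translucent Default Lit glass material for the Maple Syrup item, wrapped in the Fortnite base and pickup hologram functions.
# Noise-normal-based Fresnel drives opacity and brightens base color, while the surface normal blends flattened tiled detail with the syrup normal map.
# Emissive fakes reflections by summing a view-space sphere map, a desaturated, ambient-tinted sky cubemap and a sun/moon specular highlight.
# World position offset combines the shared LOD pickup effect with the black hole function.
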Material /Game/Creative/Items/Gameplay/Resources/Materials/M_MapleSyrup_Glass
graph {
 properties {
  bEnableResponsiveAA=true
  bScreenSpaceReflections=true
  bUseMaterialAttributes=true
  bUsedWithMeshDeformer=true
  bUsedWithNanite=true
  bUsedWithSkeletalMesh=true
  blendMode=BLEND_Translucent
  physMaterial=/Game/FortressPhysicalMaterials/Glass
  pixelDepthOffsetMode=PDOM_Legacy
  refractionMethod=RM_IndexOfRefraction
  translucencyDirectionalLightingIntensity=5
  translucencyLightingMode=TLM_VolumetricPerVertexDirectional
  usedShadingModels=Default Lit
 }
 nodes [
  Constant()
  LinearInterpolate(A=MaterialFunctionCall(Flatness=ScalarParameter() {defaultValue=1.15, group=Glass, parameterName=NormalFlatten}) {materialFunction=&FlattenNormal}.Result)
  n1 = TextureCoordinate()
  n2 = ScalarParameter() {defaultValue=6, group=Glass, parameterName=Roughness_TileRate}
  n3 = Multiply(
   A=TextureSample(UVs=Multiply(A=@n1, B=@n2)) {samplerType=SAMPLERTYPE_Masks, texture=/Game/Creative/Items/Gameplay/Resources/Textures/T_Raptor_eye_Dirt_S}.B
   B=ScalarParameter() {defaultValue=1, parameterName=Roughness_Adjustment}
  )
  n4 = TextureSample(UVs=Multiply(A=TextureCoordinate(), B=ScalarParameter() {defaultValue=6, group=Glass, parameterName=Normal_TileRate})) {samplerType=SAMPLERTYPE_Normal, texture=/Game/Creative/Items/Gameplay/Resources/Textures/T_RaptorEye_surface_N}
  n5 = ScalarParameter() {defaultValue=1.15, group=Glass, parameterName=NormalFlatten}
  n6 = TextureCoordinate()
  n7 = VectorParameter() {parameterName=WorldPostionOfBlackHole, defaultValue=<rgba -30000 25536 1952 0>}
  n8 = AppendVector(A=@n7.RGB, B=@n7.RGB)
  n9 = MaterialFunctionCall!(
   World Position
   DistanceCullFadeAnimationOffset=MaterialFunctionCall() {materialFunction=/Game/Packages/Fortress_SharedMaterials/Base_Material_Functions/PickupEffectDistanceCullOverride}.Result
   Use actor center instead of pivot point
  ) {materialFunction=/Game/Packages/Fortress_SharedMaterials/Base_Material_Functions/LODEffect}
  n10 = Add(A=@n9.Result, B=WorldPosition().XYZ)
  n11 = MaterialFunctionCall(Float2=TextureCoordinate()) {materialFunction=&BreakOutFloat2Components}
  n12 = ScalarParameter() {defaultValue=1, group=Normal, parameterName=Normal Map Tile Rate On U}
  n13 = ScalarParameter() {defaultValue=0.2, group=Normal, parameterName=Normal Map Tile Rate On V}
  n14 = TextureSampleParameter2D(UVs=AppendVector(A=Multiply(A=@n11.R, B=@n12), B=Multiply(A=@n11.G, B=@n13))) {group=Normal, parameterName=Normal, samplerType=SAMPLERTYPE_Normal, texture=/Game/Global/Textures/Default/Noise/T_Noise_Normal_01}
  n15 = Normalize(VectorInput=Multiply(A=@n14.RGB, B=VectorParameter() {group=Normal, parameterName=Normal Intensity, defaultValue=<rgba 0.1 0.1 1 0>}.RGB))
  n16 = StaticBoolParameter() {defaultValue=true, parameterName=UseNormalTexture}
  n17 = StaticSwitch(True=Transform(None=@n15), False=VertexNormalWS(), Value=@n16)
  n18 = Fresnel(ExponentIn=ScalarParameter() {defaultValue=4, parameterName=Opacity-Fresnel Exponent}, Normal=@n17)
  n19 = ComponentMask(None=Transform(None=ReflectionVectorWS()) {transformSourceType=TRANSFORMSOURCE_World, transformType=TRANSFORM_View}) {g=true, r=true}
  n20 = TextureSampleParameter2D(UVs=Multiply(A=ConstantBiasScale(None=@n19), B=Constant2Vector() {g=-1, r=1})) {group=SphereReflection, parameterName=SphereReflectionMap}
  n21 = ScalarParameter() {defaultValue=1, group=SphereReflection, parameterName=Reflection Multiplier}
  n22 = TextureSampleParameterCube!(UVs=ReflectionVectorWS() {bNormalizeCustomWorldNormal=true}, Apply View MipBias) {
   group=Cubemap
   parameterName=MaterialExpressionTextureSampleParameterCube_0
   texture=/Game/Packages/Fortress_Sky/TexturesHDR/T_hdrsky_test03_low
  }
  n23 = VectorParameter() {group=Cubemap, parameterName=Cube Map Multiplier, defaultValue=<rgba 1.5 1.5 1.5 0>}
  n24 = MaterialFunctionCall() {materialFunction=/Engine/Functions/GetPostProcessFunctions/GetAmbientCubemapTint}
  n25 = Desaturation(
   None=Multiply(A=Multiply(A=@n22.RGB, B=@n23.RGB) {constB=1.5}, B=ComponentMask(None=@n24.AmbientCubemapTint) {b=true, g=true, r=true})
   Fraction=Constant() {r=0.5}
  )
  n26 = CollectionParameter() {
   collection=&FortniteMaterialParameters
   parameterName=SunAndMoonModelDirectionalVector
  }
  n27 = DotProduct(A=ComponentMask(None=@n26) {b=true, g=true, r=true}, B=ReflectionVectorWS())
  n28 = Multiply(
   A=Power(Base=Clamp(None=@n27), Exp=ScalarParameter() {defaultValue=70, group=FakeSpec, parameterName=Spec Highlight Power}) {constExponent=20}
   B=ScalarParameter() {defaultValue=3, group=FakeSpec, parameterName=Spec Intense}
  )
  n29 = ScalarParameter() {defaultValue=1, parameterName=Diffuse-FresnelAmount}
  n30 = Fresnel(ExponentIn=ScalarParameter() {parameterName=Diffuse-FresnelExponent}, Normal=@n17)
  n31 = SetMaterialAttributes!(
   MaterialAttributes
   Base Color=Add(A=Multiply(A=@n29, B=@n30), B=VectorParameter() {parameterName=Base Color}.RGB)
   Metallic=ScalarParameter() {defaultValue=0.1, parameterName=Metallic}
   Specular=ScalarParameter() {defaultValue=1, parameterName=Specular}
   Roughness=@n3
   Emissive Color=Add(A=Multiply(A=@n20.RGB, B=@n21), B=Add(A=@n25, B=@n28))
   Opacity=LinearInterpolate(
    A=ScalarParameter() {parameterName=Opacity-Fresnel Inner Transparency}
    B=ScalarParameter() {defaultValue=0.85, parameterName=Opacity-Fresnel Outer Transparency}
    Alpha=Saturate(None=@n18)
   )
   Opacity Mask
   Normal=LinearInterpolate(
    A=MaterialFunctionCall(Normal=@n4.RGB, Flatness=@n5) {materialFunction=&FlattenNormal}.Result
    B=TextureSample(UVs=@n6) {samplerType=SAMPLERTYPE_Normal, texture=/Game/Creative/Items/Gameplay/Resources/Textures/T_CM_Prop_MapleSyrup_N}.RGB
   )
   World Position Offset=Add(
    A=MaterialFunctionCall!(WorldPostionOfBlackHole=@n8, World Position=@n10) {materialFunction=/Game/Packages/Fortnite_UniqueMaterials_1/BlackHoleFunction}.Result
    B=@n9.Result
   )
   Refraction=ScalarParameter() {defaultValue=1, parameterName=Refraction, sliderMax=2}
  ) {attributeSetTypes [
    #1
    #2
    #3
    #4
    #5
    #6
    #7
    #8
    #9
    #10
   ]}
  MaterialFunctionCall!(In=MaterialFunctionCall!(Material attributes=@n31, IsGameplayAsset=StaticBool()) {materialFunction=/Game/Packages/Fortress_SharedMaterials/Base_Material_Functions/Fortnite_BaseFunction}.Material Attributes) {materialFunction=/Game/Packages/Fortress_SharedMaterials/Base_Material_Functions/MF_Pickup_Hologram}
  Multiply(A=@n3, B=ScalarParameter() {defaultValue=0.2, parameterName=Roughness})
  Multiply()
  Reroute(None=StaticSwitch(True=@n15, False=Constant3Vector() {constant=<rgba 0 0 1 0>}, Value=@n16))
 ]
}
